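# Translucent unlit material for the Frog2LTE cosmetic wheel's animated glowing flow effect, usable on skeletal meshes.
# Two flow normal-map samples pan over radial and inverted radial coordinates, and their sum distorts the UVs.
# The squared length of the offset UVs drives both the tinted, boosted glow and the opacity, which a radial gradient fades toward the outer edge.
# Color and pan speed are parameters.
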Material /VehicleCosmetics/Wheels/Frog2LTE/Materials/M_Frog2LTE_FX
graph {
 properties {
  bAutomaticallySetUsageInEditor=false
  bUsedWithMeshDeformer=true
  bUsedWithSkeletalMesh=true
  blendMode=BLEND_Translucent
  shadingModel=MSM_Unlit
  usedShadingModels=Unlit
 }
 nodes [
  n1 = MaterialFunctionCall() {materialFunction=&VectorToRadialValue}
  n2 = Multiply(A=ScalarParameter() {defaultValue=0.25, parameterName=Pan Speed}, B=Time())
  n3 = Panner(Coordinate=@n1.Radial Coordinates, Time=@n2) {speedX=-0.1, speedY=-0.5}
  n4 = Panner(Coordinate=OneMinus(None=@n1.Radial Coordinates), Time=@n2) {speedX=-0.1, speedY=0.35}
  n5 = ComponentMask(None=Add(
    A=TextureSample(UVs=@n3) {samplerType=SAMPLERTYPE_Normal, texture=/VehicleCosmetics/Wheels/_SharedTextures/FX/Flow_03_nomips}.RGBA
    B=TextureSample(UVs=@n4) {samplerType=SAMPLERTYPE_Normal, texture=&Flow_03_nomips}.RGBA
   )) {g=true, r=true}
  n6 = Add(A=Add(A=TextureCoordinate(), B=Multiply(A=@n5) {constB=0.5})) {constB=-0.5}
  n7 = Clamp(None=DotProduct(A=@n6, B=@n6))
  Multiply(A=Multiply(A=VectorParameter() {parameterName=Color, defaultValue=<rgba 0.117 0.0091 1 1>}.RGB, B=@n7)) {constB=6}
  n8 = Subtract(A=@n7) {constB=0.001}
  n9 = ComponentMask(None=@n1.Radial Coordinates) {g=true}
  Multiply(A=Clamp(None=Multiply(A=@n8) {constB=8}), B=Clamp(None=ConstantBiasScale(None=@n9) {bias=-0.85, scale=-4}))
 ]
}
